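# Anim Blueprint for the LoanFloat backpack cosmetic that adds procedural secondary motion to its chopstick and noodle bones.
# The cached input pose feeds two differently tuned, root-relative AnimDynamics chopstick chains, selected by bIsInFrontEnd.
# The blended result runs through four serial noodle AnimDynamics nodes with angular springs, left in default simulation space, to the output.
# Additional chopstick and noodle chains are unconnected, and the EventGraph holds no logic.
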AnimBlueprint /BRCosmetics/Accessories/FORT_Backpacks/Backpack_F_MED_LoanFloat/Meshes/F_MED_LoanFloat_Pack_AnimBP
target_skeleton=&/F_MED_LoanFloat_Pack_Skeleton
graphs [
 AnimGraph {nodes [
   n1 = Anim Dynamics - Bone: dyn_chopstick_01(Alpha=1, ExternalForce="0,0,0") {
    angularDampingOverride=0.99
    angularSpringConstant=10000
    bAngularSpring=true
    bOverrideAngularDamping=true
    linearSpringConstant=1000
    simulationSpace=RootRelative
    alphaBoolBlend {blendOption=Linear}
    boundBone {boneName=dyn_chopstick_01}
    componentAppliedLinearAccClamp=<1000 1000 1000>
    componentLinearAccScale=<0.3 0.3 0.3>
    componentLinearVelScale=<0.6 0.6 0.6>
    physicsBodyDefinitions [{
      boundBone {boneName=dyn_chopstick_01}
      boxExtents=<35 25 25>
      constraintSetup {angularLimitsMax=<50 25 15>, angularLimitsMin=<-50 0 -25>, angularTarget=<0 0 0>, linearAxesMax=<0 0.2 0.2>, linearAxesMin=<-3 0 0>}
      localJointOffset=<1 0 0>
     }]
   }
   Anim Dynamics - Bone: dyn_chopstick_02(ComponentPose=@n1, Alpha=1, ExternalForce="0,0,0") {
    angularDampingOverride=0.99
    angularSpringConstant=10000
    bAngularSpring=true
    bOverrideAngularDamping=true
    linearSpringConstant=1000
    simulationSpace=RootRelative
    alphaBoolBlend {blendOption=Linear}
    boundBone {boneName=dyn_chopstick_02}
    componentAppliedLinearAccClamp=<1000 1000 1000>
    componentLinearAccScale=<0.3 0.3 0.3>
    componentLinearVelScale=<0.6 0.6 0.6>
    physicsBodyDefinitions [{
      boundBone {boneName=dyn_chopstick_02}
      boxExtents=<25 15 15>
      constraintSetup {angularLimitsMax=<50 25 15>, angularLimitsMin=<-50 0 -25>, angularTarget=<0 0 0>, linearAxesMax=<0 0.2 0.2>, linearAxesMin=<-3 0 0>}
      localJointOffset=<1 0 0>
     }]
   }
   n2 = Anim Dynamics - Bone: dyn_noodles_01(Alpha=1, ExternalForce="0,0,0") {
    angularDampingOverride=0.995
    angularSpringConstant=10000
    bAngularSpring=true
    bOverrideAngularDamping=true
    linearSpringConstant=1000
    simulationSpace=RootRelative
    alphaBoolBlend {blendOption=Linear}
    boundBone {boneName=dyn_noodles_01}
    componentAppliedLinearAccClamp=<1000 1000 1000>
    componentLinearAccScale=<0.3 0.3 0.3>
    componentLinearVelScale=<0.4 0.4 0.4>
    physicsBodyDefinitions [{
      boundBone {boneName=dyn_noodles_01}
      boxExtents=<30 30 30>
      constraintSetup {angularLimitsMax=<10 10 3>, angularLimitsMin=<-10 -10 -3>, angularTarget=<0 0 0>, linearAxesMax=<0 0.2 0.2>, linearAxesMin=<0 -0.2 -0.2>}
      localJointOffset=<0.5 0 0>
     }]
   }
   n3 = Anim Dynamics - Bone: dyn_noodles_02(ComponentPose=@n2, Alpha=1, ExternalForce="0,0,0") {
    angularDampingOverride=0.995
    angularSpringConstant=10000
    bAngularSpring=true
    bOverrideAngularDamping=true
    linearSpringConstant=1000
    simulationSpace=RootRelative
    alphaBoolBlend {blendOption=Linear}
    boundBone {boneName=dyn_noodles_02}
    componentAppliedLinearAccClamp=<1000 1000 1000>
    componentLinearAccScale=<0.5 0.5 0.5>
    componentLinearVelScale=<0.6 0.6 0.6>
    physicsBodyDefinitions [{
      boundBone {boneName=dyn_noodles_02}
      boxExtents=<20 20 20>
      constraintSetup {angularLimitsMax=<15 15 1>, angularLimitsMin=<-15 -15 -1>, angularTarget=<0 0 0>, linearAxesMax=<0 0.2 0.2>, linearAxesMin=<0 -0.2 -0.2>}
      localJointOffset=<1 0 0>
     }]
   }
   n4 = Anim Dynamics - Bone: dyn_noodles_03(ComponentPose=@n3, Alpha=1, ExternalForce="0,0,0") {
    angularDampingOverride=0.995
    angularSpringConstant=10000
    bAngularSpring=true
    bOverrideAngularDamping=true
    linearSpringConstant=1000
    simulationSpace=RootRelative
    alphaBoolBlend {blendOption=Linear}
    boundBone {boneName=dyn_noodles_03}
    componentAppliedLinearAccClamp=<1000 1000 1000>
    componentLinearAccScale=<0.4 0.4 0.4>
    componentLinearVelScale=<0.5 0.5 0.5>
    physicsBodyDefinitions [{
      boundBone {boneName=dyn_noodles_03}
      boxExtents=<40 40 40>
      constraintSetup {angularLimitsMax=<20 20 4>, angularLimitsMin=<-20 -20 -4>, angularTarget=<0 0 0>, linearAxesMax=<0 0.2 0.2>, linearAxesMin=<0 -0.2 -0.2>}
      localJointOffset=<0.5 0 0>
     }]
   }
   Anim Dynamics - Bone: dyn_noodles_04(ComponentPose=@n4, Alpha=1, ExternalForce="0,0,0") {
    angularDampingOverride=0.995
    angularSpringConstant=10000
    bAngularSpring=true
    bOverrideAngularDamping=true
    linearSpringConstant=1000
    simulationSpace=RootRelative
    alphaBoolBlend {blendOption=Linear}
    boundBone {boneName=dyn_noodles_04}
    componentAppliedLinearAccClamp=<1000 1000 1000>
    componentLinearAccScale=<0.6 0.6 0.6>
    componentLinearVelScale=<0.7 0.7 0.7>
    physicsBodyDefinitions [{
      boundBone {boneName=dyn_noodles_04}
      boxExtents=<15 15 15>
      constraintSetup {angularLimitsMax=<30 30 2>, angularLimitsMin=<-30 -30 -2>, angularTarget=<0 0 0>, linearAxesMax=<0 0.2 0.2>, linearAxesMin=<0 -0.2 -0.2>}
      localJointOffset=<1 0 0>
     }]
   }
   n5 = Anim Dynamics - Bone: dyn_chopstick_01(ComponentPose=Local To Component(LocalPose=Use cached pose 'start pose'()), Alpha=1, ExternalForce="0,0,0") {
    angularDampingOverride=0.9999
    angularSpringConstant=10000
    bAngularSpring=true
    bOverrideAngularDamping=true
    linearSpringConstant=1000
    simulationSpace=RootRelative
    alphaBoolBlend {blendOption=Linear}
    boundBone {boneName=dyn_chopstick_01}
    componentAppliedLinearAccClamp=<1000 1000 1000>
    componentLinearAccScale=<0.5 0.5 0.5>
    componentLinearVelScale=<0.6 0.6 0.6>
    physicsBodyDefinitions [{
      boundBone {boneName=dyn_chopstick_01}
      boxExtents=<25 15 15>
      constraintSetup {angularLimitsMax=<50 25 15>, angularLimitsMin=<-50 0 -25>, angularTarget=<0 0 0>, linearAxesMax=<0 0.2 0.2>, linearAxesMin=<-3 0 0>}
      localJointOffset=<1 0 0>
     }]
   }
   n6 = Anim Dynamics - Bone: dyn_chopstick_02(ComponentPose=@n5, Alpha=1, ExternalForce="0,0,0") {
    angularDampingOverride=0.9999
    angularSpringConstant=10000
    bAngularSpring=true
    bOverrideAngularDamping=true
    linearSpringConstant=1000
    simulationSpace=RootRelative
    alphaBoolBlend {blendOption=Linear}
    boundBone {boneName=dyn_chopstick_02}
    componentAppliedLinearAccClamp=<1000 1000 1000>
    componentLinearAccScale=<0.5 0.5 0.5>
    componentLinearVelScale=<0.6 0.6 0.6>
    physicsBodyDefinitions [{
      boundBone {boneName=dyn_chopstick_02}
      boxExtents=<15 7 7>
      constraintSetup {angularLimitsMax=<50 25 15>, angularLimitsMin=<-50 0 -25>, angularTarget=<0 0 0>, linearAxesMax=<0 0.2 0.2>, linearAxesMin=<-3 0 0>}
      localJointOffset=<1 0 0>
     }]
   }
   n7 = Component To Local(ComponentPose=@n6)
   n8 = Anim Dynamics - Bone: dyn_chopstick_01(ComponentPose=Local To Component(LocalPose=Use cached pose 'start pose'()), Alpha=1, ExternalForce="0,0,0") {
    angularDampingOverride=0.9999
    angularSpringConstant=10000
    bAngularSpring=true
    bOverrideAngularDamping=true
    linearSpringConstant=1000
    simulationSpace=RootRelative
    alphaBoolBlend {blendOption=Linear}
    boundBone {boneName=dyn_chopstick_01}
    componentAppliedLinearAccClamp=<1000 1000 1000>
    componentLinearAccScale=<0.4 0.4 0.4>
    componentLinearVelScale=<0.5 0.5 0.5>
    physicsBodyDefinitions [{
      boundBone {boneName=dyn_chopstick_01}
      boxExtents=<25 15 15>
      constraintSetup {angularLimitsMax=<50 25 15>, angularLimitsMin=<-50 0 -25>, angularTarget=<0 0 0>, linearAxesMax=<0 0.2 0.2>, linearAxesMin=<-3 0 0>}
      localJointOffset=<0.5 0 0>
     }]
   }
   n9 = Anim Dynamics - Bone: dyn_chopstick_02(ComponentPose=@n8, Alpha=1, ExternalForce="0,0,0") {
    angularDampingOverride=0.9999
    angularSpringConstant=10000
    bAngularSpring=true
    bOverrideAngularDamping=true
    linearSpringConstant=1000
    simulationSpace=RootRelative
    alphaBoolBlend {blendOption=Linear}
    boundBone {boneName=dyn_chopstick_02}
    componentAppliedLinearAccClamp=<1000 1000 1000>
    componentLinearAccScale=<0.4 0.4 0.4>
    componentLinearVelScale=<0.5 0.5 0.5>
    physicsBodyDefinitions [{
      boundBone {boneName=dyn_chopstick_02}
      boxExtents=<16 8 8>
      constraintSetup {angularLimitsMax=<50 25 15>, angularLimitsMin=<-50 0 -25>, angularTarget=<0 0 0>, linearAxesMax=<0 0.2 0.2>, linearAxesMin=<-3 0 0>}
      localJointOffset=<0.5 0 0>
     }]
   }
   n10 = Component To Local(ComponentPose=@n9)
   n11 = Get bIsInFrontEnd()
   n12 = Anim Dynamics - Bone: dyn_noodles_01(
    ComponentPose=Local To Component(LocalPose=Blend Poses by bool(BlendPose_0=@n10, BlendPose_1=@n7, bActiveValue=@n11) {blendType=Linear})
    Alpha=1
    ExternalForce="0,0,0"
   ) {
    angularDampingOverride=0.999
    angularSpringConstant=10000
    bAngularSpring=true
    bOverrideAngularDamping=true
    linearSpringConstant=1000
    alphaBoolBlend {blendOption=Linear}
    boundBone {boneName=dyn_noodles_01}
    componentAppliedLinearAccClamp=<1000 1000 1000>
    componentLinearAccScale=<0.3 0.3 0.3>
    componentLinearVelScale=<0.4 0.4 0.4>
    physicsBodyDefinitions [{
      boundBone {boneName=dyn_noodles_01}
      boxExtents=<30 30 30>
      constraintSetup {angularLimitsMax=<25 25 3>, angularLimitsMin=<-25 -25 -3>, angularTarget=<0 0 0>, linearAxesMax=<0 0.3 0.3>, linearAxesMin=<0 -0.3 -0.3>}
      localJointOffset=<1 0 0>
     }]
   }
   n13 = Anim Dynamics - Bone: dyn_noodles_02(ComponentPose=@n12, Alpha=1, ExternalForce="0,0,0") {
    angularDampingOverride=0.999
    angularSpringConstant=10000
    bAngularSpring=true
    bOverrideAngularDamping=true
    linearSpringConstant=1000
    alphaBoolBlend {blendOption=Linear}
    boundBone {boneName=dyn_noodles_02}
    componentAppliedLinearAccClamp=<1000 1000 1000>
    componentLinearAccScale=<0.5 0.5 0.5>
    componentLinearVelScale=<0.6 0.6 0.6>
    physicsBodyDefinitions [{
      boundBone {boneName=dyn_noodles_02}
      boxExtents=<20 20 20>
      constraintSetup {angularLimitsMax=<35 15 1>, angularLimitsMin=<-35 -35 -1>, angularTarget=<0 0 0>, linearAxesMax=<0 0.4 0.4>, linearAxesMin=<0 -0.4 -0.4>}
      localJointOffset=<1 0 0>
     }]
   }
   n14 = Anim Dynamics - Bone: dyn_noodles_03(ComponentPose=@n13, Alpha=1, ExternalForce="0,0,0") {
    angularDampingOverride=0.999
    angularSpringConstant=10000
    bAngularSpring=true
    bOverrideAngularDamping=true
    linearSpringConstant=1000
    alphaBoolBlend {blendOption=Linear}
    boundBone {boneName=dyn_noodles_03}
    componentAppliedLinearAccClamp=<1000 1000 1000>
    componentLinearAccScale=<0.4 0.4 0.4>
    componentLinearVelScale=<0.5 0.5 0.5>
    physicsBodyDefinitions [{
      boundBone {boneName=dyn_noodles_03}
      boxExtents=<40 40 40>
      constraintSetup {angularLimitsMax=<40 40 4>, angularLimitsMin=<-40 -40 -4>, angularTarget=<0 0 0>, linearAxesMax=<0 0.3 0.3>, linearAxesMin=<0 -0.3 -0.3>}
      localJointOffset=<0.5 0 0>
     }]
   }
   n15 = Anim Dynamics - Bone: dyn_noodles_04(ComponentPose=@n14, Alpha=1, ExternalForce="0,0,0") {
    angularDampingOverride=0.9999
    angularSpringConstant=10000
    bAngularSpring=true
    bOverrideAngularDamping=true
    linearSpringConstant=1000
    alphaBoolBlend {blendOption=Linear}
    boundBone {boneName=dyn_noodles_04}
    componentAppliedLinearAccClamp=<1000 1000 1000>
    componentLinearAccScale=<0.6 0.6 0.6>
    componentLinearVelScale=<0.7 0.7 0.7>
    physicsBodyDefinitions [{
      boundBone {boneName=dyn_noodles_04}
      boxExtents=<20 20 20>
      constraintSetup {angularLimitsMax=<50 50 2>, angularLimitsMin=<-50 -50 -2>, angularTarget=<0 0 0>, linearAxesMax=<0 0.3 0.3>, linearAxesMin=<0 -0.3 -0.3>}
      localJointOffset=<1 0 0>
     }]
   }
   Output Pose(Result=Component To Local(ComponentPose=@n15)) {layerGroup=None}
   Save cached pose 'start pose'(Pose=Input Pose()) {cacheName=start pose}
  ]}
 EventGraph {nodes [
   TryGetPawnOwner()
   Event BlueprintUpdateAnimation()
  ]}
]
variables []
